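# Idle Niagara effect for the MuffinLadle GasChin pickaxe cosmetic, showing purple electric glows, flashes and lightning at weapon sockets.
# Lightning001 spawns flickering, randomly chosen bolt meshes, while Flashing001 emits short-lived sprites that follow two sockets.
# Glows001 bursts two persistent glow particles with light renderers, driven by waveform flicker and scaled for time of day and front end.
# The disabled Flashing, Glows and Lightning emitters are near-duplicates with different socket offsets.
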NiagaraSystem /BRCosmetics/Weapons/FORT_Melee/Pickaxe_MuffinLadle_GasChin/FX/NS_Pickaxe_MuffinLadle_GasChin_Idle
dependencies {
 usedDataInterfaces [
  &NiagaraDataInterfaceArrayInt32
  &NiagaraDataInterfaceArrayFloat3
  &NiagaraDataInterfaceCurve
  &NiagaraDataInterfaceMeshRendererInfo
  &NiagaraDataInterfaceSocketReader
 ]
 usedDynamicInputs [&RandomRangeInt]
 usedModules [
  &self:SystemSpawnScript.SystemScriptSource.SystemScriptGraph.NiagaraNodeAssignment_0.SetVariables_#1
  &SystemState
 ]
}
emitters {
 Flashing {
  bEnabled=false
  emitterUpdateScript [
   &EmitterState(Life Cycle Mode=::NewEnumerator1, Loop Behavior=::NewEnumerator1, Loop Duration Mode=::NewEnumerator1, UseLoopDelay=false)
   &SpawnRate(SpawnRate=15, Use Spawn Probability=false)
  ]
  particleSpawnScript [
   &PartitionParticles()
   &InitializeParticle(
    Color=<rgba 1 1 1 0.5>
    Lifetime=&RandomRangeFloat(Minimum=0.25, Maximum=0.5)
    Position Mode=::NewEnumerator0
    Sprite Size Mode=::NewEnumerator3
    Uniform Sprite Size=&RandomRangeFloat(Minimum=15, Maximum=20)
    Write Material Random=true
   )
   &CameraOffset(Camera Offset Amount=2)
   &Color(Color=<rgba 0.2582 0.0437 0.6867 1>, Scale Alpha=11.9833)
  ]
  particleUpdateScript [
   &ParticleState()
   &SocketLocation(
    SamplePosOffset=true
    Sampled Position Offset=&SelectVectorFromArray(
     Direct Array Index=$Particles.PartitionParticles.Partition
     Vector Selection Array=DataInterface() {propertyValues {FloatData [<1 0 2>, <1 0 -2>]}}
    )
    Socket Index=&SelectIntFromArray(Direct Array Index=$Particles.PartitionParticles.Partition, Int Selection Array=DataInterface() {propertyValues {IntData [1,2]}})
    Socket Reader=$System.Socket Reader
    Socket Sampling Mode=::NewEnumerator2
    VelocitySpeedLimit=false
   )
   /SetVariables(Particles.2LightPos:NiagaraPosition=&Add_Vector(A=$Output.SocketLocation.SampledPosition, B=<-0.25 0 32>)) {isSetParameters=true, module=SetVariables_#2}
   &ScaleSpriteSize(Uniform Curve Sprite Scale=DataInterface() {propertyValues {Curve {keys [{tangentMode=RCTM_User, time=0, value=0}, {tangentMode=RCTM_User, time=0.3, value=1}]}}})
   &SolveForcesAndVelocity(Clamp Velocity=false, Limit Acceleration=false)
   &ScaleColor(
    Scale Alpha=&FloatFromCurve(
     FloatCurve=DataInterface() {propertyValues {Curve {keys [
         {tangentMode=RCTM_User, time=0, value=1}
         {arriveTangent=-0.0405, interpMode=RCIM_Cubic, leaveTangent=-0.0405, tangentMode=RCTM_User, time=0.7009, value=1}
         {arriveTangent=-8.0089, interpMode=RCIM_Cubic, leaveTangent=-8.0089, tangentMode=RCTM_User, time=1, value=0}
        ]}}}
     Scale Curve=0.5
    )
    Scale RGB=&Lerp_Vector(
     A=&VectorFromFloat(Value=&FloatFromCurve(
       FloatCurve=DataInterface() {propertyValues {Curve {keys [<at 0 1>,<at 0.1 1>,<at 0.2 2>,<at 0.3 1>,<at 0.6 1>,<at 0.7 2>,<at 0.8 1>,<at 1 1>]}}}
       Scale Curve=&MaskFloatBySpawnGroup(Masked Float=1, Passthrough Float=1)
      ))
     B=&VectorFromFloat(Value=&RandomRangeFloat(Minimum=0.3, Maximum=1, Evaluation Type=::NewEnumerator1))
     Alpha=&FloatFromCurve(
      FloatCurve=DataInterface() {propertyValues {Curve {keys [
          {interpMode=RCIM_Cubic, leaveTangent=2.5294, tangentMode=RCTM_Break, time=0, value=0}
          {arriveTangent=-0.2166, interpMode=RCIM_Cubic, tangentMode=RCTM_Break, time=0.4, value=1}
         ]}}}
      Scale Curve=0.5
     )
    )
    ScaleA=true
    ScaleRGB=true
   ) {module=ScaleColor001}
  ]
  rendererClasses [&NiagaraSpriteRendererProperties]
  renderers [
   {rendererClass=&NiagaraSpriteRendererProperties}
   {rendererClass=&NiagaraSpriteRendererProperties, rendererIndex=1}
  ]
 }
 Flashing001 {
  emitterUpdateScript [
   &EmitterState(Life Cycle Mode=::NewEnumerator1, Loop Behavior=::NewEnumerator1, Loop Duration Mode=::NewEnumerator1, UseLoopDelay=false)
   &SpawnRate(SpawnRate=15, Use Spawn Probability=false)
  ]
  particleSpawnScript [
   &PartitionParticles()
   &InitializeParticle(
    Color=<rgba 1 1 1 0.5>
    Lifetime=&RandomRangeFloat(Minimum=0.25, Maximum=0.5)
    Position Mode=::NewEnumerator0
    Sprite Size Mode=::NewEnumerator3
    Uniform Sprite Size=&RandomRangeFloat(Minimum=15, Maximum=20)
    Write Material Random=true
   )
   &CameraOffset(Camera Offset Amount=2)
   &Color(Color=<rgba 0.2582 0.0437 0.6867 1>, Scale Alpha=11.9833)
  ]
  particleUpdateScript [
   &ParticleState()
   &SocketLocation(
    SamplePosOffset=true
    Sampled Position Offset=&SelectVectorFromArray(
     Direct Array Index=$Particles.PartitionParticles.Partition
     Vector Selection Array=DataInterface() {propertyValues {FloatData [<0 0 35>, <0 0 -35>]}}
    )
    Socket Index=&SelectIntFromArray(Direct Array Index=$Particles.PartitionParticles.Partition, Int Selection Array=DataInterface() {propertyValues {IntData [1,2]}})
    Socket Reader=$System.Socket Reader
    Socket Sampling Mode=::NewEnumerator2
    VelocitySpeedLimit=false
   )
   /SetVariables(Particles.2LightPos:NiagaraPosition=&Add_Vector(A=$Output.SocketLocation.SampledPosition, B=<-0.25 0 32>)) {isSetParameters=true, module=SetVariables_#2}
   &ScaleSpriteSize(Uniform Curve Sprite Scale=DataInterface() {propertyValues {Curve {keys [{tangentMode=RCTM_User, time=0, value=0}, {tangentMode=RCTM_User, time=0.3, value=1}]}}})
   &SolveForcesAndVelocity(Clamp Velocity=false, Limit Acceleration=false)
   &ScaleColor(
    Scale Alpha=&FloatFromCurve(
     FloatCurve=DataInterface() {propertyValues {Curve {keys [
         {tangentMode=RCTM_User, time=0, value=1}
         {arriveTangent=-0.0405, interpMode=RCIM_Cubic, leaveTangent=-0.0405, tangentMode=RCTM_User, time=0.7009, value=1}
         {arriveTangent=-8.0089, interpMode=RCIM_Cubic, leaveTangent=-8.0089, tangentMode=RCTM_User, time=1, value=0}
        ]}}}
     Scale Curve=0.5
    )
    Scale RGB=&Lerp_Vector(
     A=&VectorFromFloat(Value=&FloatFromCurve(
       FloatCurve=DataInterface() {propertyValues {Curve {keys [<at 0 1>,<at 0.1 1>,<at 0.2 2>,<at 0.3 1>,<at 0.6 1>,<at 0.7 2>,<at 0.8 1>,<at 1 1>]}}}
       Scale Curve=&MaskFloatBySpawnGroup(Masked Float=1, Passthrough Float=1)
      ))
     B=&VectorFromFloat(Value=&RandomRangeFloat(Minimum=0.3, Maximum=1, Evaluation Type=::NewEnumerator1))
     Alpha=&FloatFromCurve(
      FloatCurve=DataInterface() {propertyValues {Curve {keys [
          {interpMode=RCIM_Cubic, leaveTangent=2.5294, tangentMode=RCTM_Break, time=0, value=0}
          {arriveTangent=-0.2166, interpMode=RCIM_Cubic, tangentMode=RCTM_Break, time=0.4, value=1}
         ]}}}
      Scale Curve=0.5
     )
    )
    ScaleA=true
    ScaleRGB=true
   ) {module=ScaleColor001}
  ]
  rendererClasses [&NiagaraSpriteRendererProperties]
  renderers [
   {rendererClass=&NiagaraSpriteRendererProperties}
   {rendererClass=&NiagaraSpriteRendererProperties, rendererIndex=1}
  ]
 }
 Glows {
  bEnabled=false
  emitterSpawnScript [
   /SetVariables(Emitter.LightPosition:NiagaraPosition=<0 0 55>) {isSetParameters=true, module=SetVariables_#3}
   /SetVariables(Emitter.LightPosition2:NiagaraPosition=<0 0 -55>) {isSetParameters=true, module=SetVariables_#4}
  ]
  emitterUpdateScript [
   /SetVariables(Emitter.LightColor:LinearColor=&MultiplyLinearColor(
     Linear Color=<rgba 1.142 0.1228 3 0.142>
     Scale Factor=&MakeVector4_FromFloat(Float=&Multiply_Float(
       A=$Emitter.Flicker
       B=&FN_FrontEndFloatModifier(
        Gameplay=&Multiply_Float(A=2, B=&FNTimeOfDayScalar(Night Scale=0.73, Morning Scale=0.85, Evening Scale=0.85))
        Front End=0.7
       )
      ))
    )) {isSetParameters=true, module=SetVariables_#5}
   /Niagara/Noise/Waveform(
    EnableExponent1=false
    GlobalClampMax=false
    GlobalClampMin=false
    UseInput1=false
    UsePhase1=false
    "Waveform [1]":ENiagara_Waveforms=::NewEnumerator12
    "[1] Amplitude Min/Max":Vector2f=<0.75 1>
    "[1] Frequency"=7
    "[1] Phase Scale"=2
   )
   /SetVariables(
    Emitter.Flicker:NiagaraFloat=&Add_Float(
     A=&Multiply_Float(A=$Emitter.Waveform.Waveform, B=&Lerp_Float(A=1, B=1, Alpha=$System.Timeline.Time))
     B=0
    )
    Emitter.FlickerHeavy:NiagaraFloat=&Add_Float(A=&Multiply_Float(A=$Emitter.Waveform.Waveform, B=2), B=0.8)
   ) {isSetParameters=true, module=SetVariables_#6}
   &EmitterState(
    Inactive Response=::NewEnumerator1
    Life Cycle Mode=::NewEnumerator1
    Loop Behavior=::NewEnumerator1
    Loop Duration Mode=::NewEnumerator1
    UseLoopDelay=false
   )
   &SpawnBurst_Instantaneous(Spawn Count=2, Use Spawn Probability=false, UseLoopCountLimit=false)
  ]
  particleSpawnScript [
   &PartitionParticles(Particles Per Partition=1, Partition Mode:ENiagara_PartitionMode_v2=::NewEnumerator1)
   &InitializeParticle(Sprite Size Mode=::NewEnumerator3, Uniform Sprite Size=120, UsePositionOffset=false, Write Material Random=true)
   &FN_SetSpawnGroup(Number of Spawn Groups=2)
   &DynamicMaterialParameters(
    Index 0 Param 1=0.2
    Index 0 Param 2=1
    Module.Index0Param0Write=true
    Module.Index0Param1Write=true
    Module.Index0Param2Write=true
    Module.Index0Param3Write=true
   )
   &CameraOffset(Camera Offset Amount=11)
   &SocketLocation(
    SamplePosOffset=true
    Sampled Position Offset=&SelectVectorFromArray(
     Direct Array Index=&SpawnGroup()
     Vector Selection Array=DataInterface() {propertyValues {FloatData [<1 0 -3>, <1 0 -5>]}}
    )
    Socket Index=&SelectIntFromArray(Direct Array Index=&SpawnGroup(), Int Selection Array=DataInterface() {propertyValues {IntData [1,2]}})
    Socket Reader=$System.Socket Reader
    Socket Sampling Mode=::NewEnumerator2
    VelocitySpeedLimit=false
   )
   /SetVariables(
    Particles.CoreGlowPosition:NiagaraPosition=$Output.SocketLocation.SampledPosition
    Particles.CoreGlowSize:Vector2f=<12 40>
    Particles.DarkGlowPosition:NiagaraPosition=&Add_Vector(A=Linked() {linkedVariable {name=Output.SocketLocation.SampledPosition, type=NiagaraPosition}}, B=<0 0 22.5>)
    Particles.DarkGlowSize:Vector2f=&Vector2DFromFloat(Value=75)
    Particles.SpriteAlignment=<0 0 1>
   ) {isSetParameters=true, module=SetVariables_#7}
   /SetVariables(Particles.TubeOverlayPosition:NiagaraPosition=&SelectVectorFromArray(
     Direct Array Index=&SpawnGroup()
     Vector Selection Array=DataInterface() {propertyValues {FloatData [<0 0 66.25>, <0 0 0>]}}
    )) {isSetParameters=true, module=SetVariables_#8}
  ]
  particleUpdateScript [
   &ParticleState(Kill Particles When Lifetime Has Elapsed=false, Let Infinitely Lived Particles Die When Emitter Deactivates=true)
   /SetVariables(
    Particles.CoreGlowColor:LinearColor=&MultiplyLinearColor(
     Linear Color=<rgba 1.142 0.1228 3 0.25>
     Scale Factor=&MakeVector4_FromFloat(Float=&Multiply_Float(
       A=$Emitter.Flicker
       B=&FN_FrontEndFloatModifier(Gameplay=&FNTimeOfDayScalar(Night Scale=1, Morning Scale=1, Evening Scale=1), Front End=0.7)
      ))
    )
    Particles.DarkGlowColor:LinearColor=&MultiplyLinearColor(
     Linear Color=<rgba 0.2854 0.0292 0.564 0.195>
     Scale Factor=&MakeVector4_FromFloat(Float=&Multiply_Float(
       A=$Emitter.Flicker
       B=&FN_FrontEndFloatModifier(Gameplay=&FNTimeOfDayScalar(Night Scale=1, Morning Scale=1, Evening Scale=1), Front End=0.7)
      ))
    )
    Particles.TubeOverlayColor:LinearColor=&MultiplyLinearColor(
     Linear Color=<rgba 0.2513 0.06 0.6 0.142>
     Scale Factor=&MakeVector4_FromFloat(Float=&Multiply_Float(
       A=$Emitter.Flicker
       B=&FN_FrontEndFloatModifier(
        Gameplay=&Multiply_Float(
         A=&FNTimeOfDayScalar(Night Scale=1, Morning Scale=4, Evening Scale=1)
         B=&Lerp_Float(A=0.5, B=3, Alpha=$System.Timeline.Time)
        )
        Front End=0.7
       )
      ))
    )
    Particles.TubeOverlayOutlineColor:LinearColor=&MultiplyLinearColor(
     Linear Color=<rgba 0.2513 0.06 0.6 0.142>
     Scale Factor=&MakeVector4_FromFloat(Float=&Multiply_Float(
       A=$Emitter.Flicker
       B=&FN_FrontEndFloatModifier(
        Gameplay=&Multiply_Float(
         A=&FNTimeOfDayScalar(Night Scale=1, Morning Scale=1, Evening Scale=1)
         B=&Lerp_Float(A=2, B=3, Alpha=$System.Timeline.Time)
        )
        Front End=0.7
       )
      ))
    )
   ) {isSetParameters=true, module=SetVariables_#9}
  ]
  rendererClasses [&NiagaraSpriteRendererProperties, &NiagaraLightRendererProperties]
  renderers [
   {rendererClass=&NiagaraSpriteRendererProperties}
   {rendererClass=&NiagaraSpriteRendererProperties, rendererIndex=1}
   {rendererClass=&NiagaraLightRendererProperties, rendererIndex=2}
   {rendererClass=&NiagaraLightRendererProperties, rendererIndex=3}
  ]
 }
 Glows001 {
  emitterSpawnScript [
   /SetVariables(Emitter.LightPosition:NiagaraPosition=<0 0 55>) {isSetParameters=true, module=SetVariables_#3}
   /SetVariables(Emitter.LightPosition2:NiagaraPosition=<0 0 -55>) {isSetParameters=true, module=SetVariables_#4}
  ]
  emitterUpdateScript [
   /SetVariables(Emitter.LightColor:LinearColor=&MultiplyLinearColor(
     Linear Color=<rgba 1.142 0.1228 3 0.142>
     Scale Factor=&MakeVector4_FromFloat(Float=&Multiply_Float(
       A=$Emitter.Flicker
       B=&FN_FrontEndFloatModifier(
        Gameplay=&Multiply_Float(A=2, B=&FNTimeOfDayScalar(Night Scale=0.73, Morning Scale=0.85, Evening Scale=0.85))
        Front End=0.7
       )
      ))
    )) {isSetParameters=true, module=SetVariables_#5}
   &Waveform(
    EnableExponent1=false
    GlobalClampMax=false
    GlobalClampMin=false
    UseInput1=false
    UsePhase1=false
    "Waveform [1]":ENiagara_Waveforms=::NewEnumerator12
    "[1] Amplitude Min/Max":Vector2f=<0.75 1>
    "[1] Frequency"=7
    "[1] Phase Scale"=2
   )
   /SetVariables(
    Emitter.Flicker:NiagaraFloat=&Add_Float(
     A=&Multiply_Float(A=$Emitter.Waveform.Waveform, B=&Lerp_Float(A=1, B=1, Alpha=$System.Timeline.Time))
     B=0
    )
    Emitter.FlickerHeavy:NiagaraFloat=&Add_Float(A=&Multiply_Float(A=$Emitter.Waveform.Waveform, B=2), B=0.8)
   ) {isSetParameters=true, module=SetVariables_#6}
   &EmitterState(
    Inactive Response=::NewEnumerator1
    Life Cycle Mode=::NewEnumerator1
    Loop Behavior=::NewEnumerator1
    Loop Duration Mode=::NewEnumerator1
    UseLoopDelay=false
   )
   &SpawnBurst_Instantaneous(Spawn Count=2, Use Spawn Probability=false, UseLoopCountLimit=false)
  ]
  particleSpawnScript [
   &PartitionParticles(Particles Per Partition=1, Partition Mode:ENiagara_PartitionMode_v2=::NewEnumerator1)
   &InitializeParticle(Sprite Size Mode=::NewEnumerator3, Uniform Sprite Size=120, UsePositionOffset=false, Write Material Random=true)
   &FN_SetSpawnGroup(Number of Spawn Groups=2)
   &DynamicMaterialParameters(
    Index 0 Param 1=0.2
    Index 0 Param 2=1
    Module.Index0Param0Write=true
    Module.Index0Param1Write=true
    Module.Index0Param2Write=true
    Module.Index0Param3Write=true
   )
   &CameraOffset(Camera Offset Amount=11)
   &SocketLocation(
    SamplePosOffset=true
    Sampled Position Offset=&SelectVectorFromArray(
     Direct Array Index=&SpawnGroup()
     Vector Selection Array=DataInterface() {propertyValues {FloatData [<1 0 35>, <1 0 -35>]}}
    )
    Socket Index=&SelectIntFromArray(Direct Array Index=&SpawnGroup(), Int Selection Array=DataInterface() {propertyValues {IntData [1,2]}})
    Socket Reader=$System.Socket Reader
    Socket Sampling Mode=::NewEnumerator2
    VelocitySpeedLimit=false
   )
   /SetVariables(
    Particles.CoreGlowPosition:NiagaraPosition=$Output.SocketLocation.SampledPosition
    Particles.CoreGlowSize:Vector2f=<12 40>
    Particles.DarkGlowPosition:NiagaraPosition=&Add_Vector(A=Linked() {linkedVariable {name=Output.SocketLocation.SampledPosition, type=NiagaraPosition}}, B=<0 0 22.5>)
    Particles.DarkGlowSize:Vector2f=&Vector2DFromFloat(Value=75)
    Particles.SpriteAlignment=<0 0 1>
   ) {isSetParameters=true, module=SetVariables_#7}
   /SetVariables(Particles.TubeOverlayPosition:NiagaraPosition=&SelectVectorFromArray(
     Direct Array Index=&SpawnGroup()
     Vector Selection Array=DataInterface() {propertyValues {FloatData [<0 0 66.25>, <0 0 0>]}}
    )) {isSetParameters=true, module=SetVariables_#8}
  ]
  particleUpdateScript [
   &ParticleState(Kill Particles When Lifetime Has Elapsed=false, Let Infinitely Lived Particles Die When Emitter Deactivates=true)
   /SetVariables(
    Particles.CoreGlowColor:LinearColor=&MultiplyLinearColor(
     Linear Color=<rgba 1.142 0.1228 3 0.25>
     Scale Factor=&MakeVector4_FromFloat(Float=&Multiply_Float(
       A=$Emitter.Flicker
       B=&FN_FrontEndFloatModifier(Gameplay=&FNTimeOfDayScalar(Night Scale=1, Morning Scale=1, Evening Scale=1), Front End=0.7)
      ))
    )
    Particles.DarkGlowColor:LinearColor=&MultiplyLinearColor(
     Linear Color=<rgba 0.2854 0.0292 0.564 0.195>
     Scale Factor=&MakeVector4_FromFloat(Float=&Multiply_Float(
       A=$Emitter.Flicker
       B=&FN_FrontEndFloatModifier(Gameplay=&FNTimeOfDayScalar(Night Scale=1, Morning Scale=1, Evening Scale=1), Front End=0.7)
      ))
    )
    Particles.TubeOverlayColor:LinearColor=&MultiplyLinearColor(
     Linear Color=<rgba 0.2513 0.06 0.6 0.142>
     Scale Factor=&MakeVector4_FromFloat(Float=&Multiply_Float(
       A=$Emitter.Flicker
       B=&FN_FrontEndFloatModifier(
        Gameplay=&Multiply_Float(
         A=&FNTimeOfDayScalar(Night Scale=1, Morning Scale=4, Evening Scale=1)
         B=&Lerp_Float(A=0.5, B=3, Alpha=$System.Timeline.Time)
        )
        Front End=0.7
       )
      ))
    )
    Particles.TubeOverlayOutlineColor:LinearColor=&MultiplyLinearColor(
     Linear Color=<rgba 0.2513 0.06 0.6 0.142>
     Scale Factor=&MakeVector4_FromFloat(Float=&Multiply_Float(
       A=$Emitter.Flicker
       B=&FN_FrontEndFloatModifier(
        Gameplay=&Multiply_Float(
         A=&FNTimeOfDayScalar(Night Scale=1, Morning Scale=1, Evening Scale=1)
         B=&Lerp_Float(A=2, B=3, Alpha=$System.Timeline.Time)
        )
        Front End=0.7
       )
      ))
    )
   ) {isSetParameters=true, module=SetVariables_#9}
  ]
  rendererClasses [&NiagaraSpriteRendererProperties, &NiagaraLightRendererProperties]
  renderers [
   {rendererClass=&NiagaraSpriteRendererProperties}
   {rendererClass=&NiagaraSpriteRendererProperties, rendererIndex=1}
   {rendererClass=&NiagaraLightRendererProperties, rendererIndex=2}
   {rendererClass=&NiagaraLightRendererProperties, rendererIndex=3}
  ]
 }
 Lightning {
  bEnabled=false
  emitterUpdateScript [
   &EmitterState(
    Inactive Response=::NewEnumerator1
    Life Cycle Mode=::NewEnumerator1
    Loop Behavior=::NewEnumerator1
    Loop Duration Mode=::NewEnumerator1
    UseLoopDelay=false
   )
   &SpawnRate(Spawn Probability=0.85, SpawnRate=18, Use Spawn Probability=true)
   &SpawnBurst_Instantaneous(Spawn Count=8, Use Spawn Probability=false, UseLoopCountLimit=false)
  ]
  particleSpawnScript [
   &PartitionParticles()
   &InitializeParticle(
    Color=<rgba 0.4018 0.03 1 1>
    Lifetime Max=1.1
    Lifetime Min=0.4
    Lifetime Mode=::NewEnumerator1
    Mesh Scale=<0.8 0.8 1.2>
    Mesh Scale Mode=::NewEnumerator4
    Position Offset=<15 0 73.9354>
    UsePositionOffset=false
    Write Material Random=true
   )
   &InitialMeshOrientation(
    Mesh Orientation Mode=::NewEnumerator2
    Orientation Vector=&RandomRangeVector(Minimum=<-1 -1 0>, Maximum=<1 1 0>)
    Rotation=<0 0 1>
    Use Rotation Vector=true
   )
   /SetVariables(Particles.MeshIndex=&MakeIntFromFloat(Float=&FloatFromCurve(
      FloatCurve=DataInterface() {propertyValues {Curve {keys [
          {arriveTangent=0.6236, interpMode=RCIM_Cubic, tangentMode=RCTM_Break, time=0.25, value=0}
          {arriveTangent=22.9944, interpMode=RCIM_Cubic, tangentMode=RCTM_Break, time=1, value=5}
         ]}}}
      CurveIndex=&RandomRangeFloat(Minimum=0, Maximum=1)
     ))) {isSetParameters=true, module=SetVariables_#10}
   &SocketLocation(
    SamplePosOffset=true
    Sampled Position Offset=&SelectVectorFromArray(
     Direct Array Index=$Particles.PartitionParticles.Partition
     Vector Selection Array=DataInterface() {propertyValues {FloatData [<1 0 0>, <1 0 -1>]}}
    )
    Socket Index=&SelectIntFromArray(Direct Array Index=$Particles.PartitionParticles.Partition, Int Selection Array=DataInterface() {propertyValues {IntData [0,3]}})
    Socket Reader=$System.Socket Reader
    Socket Sampling Mode=::NewEnumerator2
    VelocitySpeedLimit=false
   )
  ]
  particleUpdateScript [
   &ParticleState()
   /SetVariables(Particles.LightningFlicker:NiagaraFloat=/Niagara/DynamicInputs/Helpers/Waveform(
     Global Input Position=&RandomRangeFloat(Minimum=0, Maximum=1)
     Global Phase=$Particles.Age
     "Waveform [1]"=::NewEnumerator12
     "[1] Phase Scale"=66
     "[1] Frequency"=1
     "[1] Amplitude Min/Max"=<0.4 1>
    )) {isSetParameters=true, module=SetVariables_#11}
   &ScaleColor(
    Scale Alpha=&FloatFromCurve(
     FloatCurve=DataInterface() {propertyValues {Curve {keys [
         {arriveTangent=9.0725, interpMode=RCIM_Cubic, leaveTangent=9.0725, tangentMode=RCTM_User, time=0, value=0}
         {arriveTangent=-2.5249, interpMode=RCIM_Cubic, leaveTangent=-2.5249, tangentMode=RCTM_User, time=0.1408, value=0.9847}
         {arriveTangent=-0.0405, interpMode=RCIM_Cubic, leaveTangent=-0.0405, tangentMode=RCTM_User, time=1, value=0}
        ]}}}
     Scale Curve=$Particles.LightningFlicker
    )
    Scale RGB=&VectorFromFloat(Value=&FloatFromCurve(
      FloatCurve=DataInterface() {propertyValues {Curve {keys [
          {arriveTangent=9.0725, interpMode=RCIM_Cubic, leaveTangent=9.0725, tangentMode=RCTM_Break, time=0, value=0}
          {arriveTangent=-2.5249, interpMode=RCIM_Cubic, leaveTangent=-2.5249, tangentMode=RCTM_Break, time=0.1408, value=1}
          {arriveTangent=-0.0405, interpMode=RCIM_Cubic, leaveTangent=-0.0405, tangentMode=RCTM_User, time=1, value=0}
         ]}}}
      Scale Curve=$Particles.LightningFlicker
     ))
    ScaleA=true
    ScaleRGB=true
   )
   /SetVariables(
    Particles.ColorCore:LinearColor=&MultiplyLinearColor(
     Linear Color=$Particles.Color
     Scale Factor=&MakeFloat4FromVectorAndFloat("Vector (RGB)"=&VectorFromFloat(Value=33), "Float (Alpha)"=1)
    )
    Particles.ColorGlow:LinearColor=&MultiplyLinearColor(
     Linear Color=$Particles.Color
     Scale Factor=&MakeFloat4FromVectorAndFloat("Vector (RGB)"=&VectorFromFloat(Value=3), "Float (Alpha)"=1)
    )
    Particles.DynamicCore:Vector4f=<1 6 2 0>
    Particles.DynamicGlow:Vector4f=<0.5 3 4 0>
    Particles.LightningScale:NiagaraFloat=&FloatFromCurve(FloatCurve=DataInterface() {propertyValues {Curve {keys [
         {arriveTangent=2.7028, interpMode=RCIM_Constant, time=0, value=0.75}
         {interpMode=RCIM_Cubic, tangentMode=RCTM_User, time=0.5, value=1}
         {arriveTangent=-2.6499, tangentMode=RCTM_User, time=1, value=0.5}
        ]}}})
   ) {isSetParameters=true, module=SetVariables_#12}
   &ScaleMeshSize(Scale Factor=&MakeVector(X=$Particles.LightningScale, Y=$Particles.LightningScale, Z=1))
   &UpdateMeshOrientation(
    Rotation Rate=&RandomRangeFloat(Minimum=-0.5, Maximum=0.5)
    Rotation Vector=&RotateVectorByQuaternion(Vector To Rotate=<0 0 1>, Quaternion=$StackContext.SkeletalMeshLocation.SampledRotation)
   )
  ]
  rendererClasses [&NiagaraMeshRendererProperties]
  renderers [
   {rendererClass=&NiagaraMeshRendererProperties, rendererIndex=0}
   {rendererClass=&NiagaraMeshRendererProperties, rendererIndex=1}
  ]
 }
 Lightning001 {
  emitterUpdateScript [
   &EmitterState(
    Inactive Response=::NewEnumerator1
    Life Cycle Mode=::NewEnumerator1
    Loop Behavior=::NewEnumerator1
    Loop Duration Mode=::NewEnumerator1
    UseLoopDelay=false
   )
   &SpawnRate(Spawn Probability=0.85, SpawnRate=18, Use Spawn Probability=true)
   &SpawnBurst_Instantaneous(Spawn Count=8, Use Spawn Probability=false, UseLoopCountLimit=false)
  ]
  particleSpawnScript [
   &PartitionParticles()
   &InitializeParticle(
    Color=<rgba 0.4018 0.03 1 1>
    Lifetime Max=1.1
    Lifetime Min=0.4
    Lifetime Mode=::NewEnumerator1
    Mesh Scale=<0.8 0.8 1.2>
    Mesh Scale Mode=::NewEnumerator4
    Position Offset=<15 0 73.9354>
    UsePositionOffset=false
    Write Material Random=true
   )
   &InitialMeshOrientation(
    Mesh Orientation Mode=::NewEnumerator2
    Orientation Vector=&RandomRangeVector(Minimum=<-1 -1 0>, Maximum=<1 1 0>)
    Rotation=<0 0 1>
    Use Rotation Vector=true
   )
   /SetVariables(Particles.MeshIndex=&MakeIntFromFloat(Float=&FloatFromCurve(
      FloatCurve=DataInterface() {propertyValues {Curve {keys [
          {arriveTangent=0.6236, interpMode=RCIM_Cubic, tangentMode=RCTM_Break, time=0.25, value=0}
          {arriveTangent=22.9944, interpMode=RCIM_Cubic, tangentMode=RCTM_Break, time=1, value=5}
         ]}}}
      CurveIndex=&RandomRangeFloat(Minimum=0, Maximum=1)
     ))) {isSetParameters=true, module=SetVariables_#10}
   &SocketLocation(
    SamplePosOffset=true
    Sampled Position Offset=&SelectVectorFromArray(
     Direct Array Index=$Particles.PartitionParticles.Partition
     Vector Selection Array=DataInterface() {propertyValues {FloatData [<1 0 -92>, <1 0 32>]}}
    )
    Socket Index=&SelectIntFromArray(Direct Array Index=$Particles.PartitionParticles.Partition, Int Selection Array=DataInterface() {propertyValues {IntData [0,3]}})
    Socket Reader=$System.Socket Reader
    Socket Sampling Mode=::NewEnumerator2
    VelocitySpeedLimit=false
   )
  ]
  particleUpdateScript [
   &ParticleState()
   /SetVariables(Particles.LightningFlicker:NiagaraFloat=&Waveform(
     Global Input Position=&RandomRangeFloat(Minimum=0, Maximum=1)
     Global Phase=$Particles.Age
     "Waveform [1]"=::NewEnumerator12
     "[1] Phase Scale"=66
     "[1] Frequency"=1
     "[1] Amplitude Min/Max"=<0.4 1>
    )) {isSetParameters=true, module=SetVariables_#11}
   &ScaleColor(
    Scale Alpha=&FloatFromCurve(
     FloatCurve=DataInterface() {propertyValues {Curve {keys [
         {arriveTangent=9.0725, interpMode=RCIM_Cubic, leaveTangent=9.0725, tangentMode=RCTM_User, time=0, value=0}
         {arriveTangent=-2.5249, interpMode=RCIM_Cubic, leaveTangent=-2.5249, tangentMode=RCTM_User, time=0.1408, value=0.9847}
         {arriveTangent=-0.0405, interpMode=RCIM_Cubic, leaveTangent=-0.0405, tangentMode=RCTM_User, time=1, value=0}
        ]}}}
     Scale Curve=$Particles.LightningFlicker
    )
    Scale RGB=&VectorFromFloat(Value=&FloatFromCurve(
      FloatCurve=DataInterface() {propertyValues {Curve {keys [
          {arriveTangent=9.0725, interpMode=RCIM_Cubic, leaveTangent=9.0725, tangentMode=RCTM_Break, time=0, value=0}
          {arriveTangent=-2.5249, interpMode=RCIM_Cubic, leaveTangent=-2.5249, tangentMode=RCTM_Break, time=0.1408, value=1}
          {arriveTangent=-0.0405, interpMode=RCIM_Cubic, leaveTangent=-0.0405, tangentMode=RCTM_User, time=1, value=0}
         ]}}}
      Scale Curve=$Particles.LightningFlicker
     ))
    ScaleA=true
    ScaleRGB=true
   )
   /SetVariables(
    Particles.ColorCore:LinearColor=&MultiplyLinearColor(
     Linear Color=$Particles.Color
     Scale Factor=&MakeFloat4FromVectorAndFloat("Vector (RGB)"=&VectorFromFloat(Value=33), "Float (Alpha)"=1)
    )
    Particles.ColorGlow:LinearColor=&MultiplyLinearColor(
     Linear Color=$Particles.Color
     Scale Factor=&MakeFloat4FromVectorAndFloat("Vector (RGB)"=&VectorFromFloat(Value=3), "Float (Alpha)"=1)
    )
    Particles.DynamicCore:Vector4f=<1 6 2 0>
    Particles.DynamicGlow:Vector4f=<0.5 3 4 0>
    Particles.LightningScale:NiagaraFloat=&FloatFromCurve(FloatCurve=DataInterface() {propertyValues {Curve {keys [
         {arriveTangent=2.7028, interpMode=RCIM_Constant, time=0, value=0.75}
         {interpMode=RCIM_Cubic, tangentMode=RCTM_User, time=0.5, value=1}
         {arriveTangent=-2.6499, tangentMode=RCTM_User, time=1, value=0.5}
        ]}}})
   ) {isSetParameters=true, module=SetVariables_#12}
   &ScaleMeshSize(Scale Factor=&MakeVector(X=$Particles.LightningScale, Y=$Particles.LightningScale, Z=1))
   &UpdateMeshOrientation(
    Rotation Rate=&RandomRangeFloat(Minimum=-0.5, Maximum=0.5)
    Rotation Vector=&RotateVectorByQuaternion(Vector To Rotate=<0 0 1>, Quaternion=$StackContext.SkeletalMeshLocation.SampledRotation)
   )
  ]
  rendererClasses [&NiagaraMeshRendererProperties]
  renderers [
   {rendererClass=&NiagaraMeshRendererProperties, rendererIndex=0}
   {rendererClass=&NiagaraMeshRendererProperties, rendererIndex=1}
  ]
 }
}
summary {
 systemName=NS_Pickaxe_MuffinLadle_GasChin_Idle
 emitters [
  {bEnabled=false, emitterName=Flashing, rendererClasses [&NiagaraSpriteRendererProperties]}
  {bEnabled=false, emitterName=Lightning, rendererClasses [&NiagaraMeshRendererProperties]}
  {
   bEnabled=false
   emitterName=Glows
   rendererClasses [&NiagaraSpriteRendererProperties, &NiagaraLightRendererProperties]
  }
  {emitterName=Lightning001, rendererClasses [&NiagaraMeshRendererProperties]}
  {emitterName=Glows001, rendererClasses [&NiagaraSpriteRendererProperties, &NiagaraLightRendererProperties]}
  {emitterName=Flashing001, rendererClasses [&NiagaraSpriteRendererProperties]}
 ]
 userVariables [
  User.bIsFrontEnd {type=NiagaraBool}
  User.bIsFrontEndPreview {type=NiagaraBool}
 ]
}
